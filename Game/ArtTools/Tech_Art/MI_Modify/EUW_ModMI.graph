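# Editor utility widget that resets or finds Material Instance Constants overriding parameters named in the editable ResetParams set.
# It scans scalar, texture and vector overrides of each listed object, caching unlisted ones in maps and collecting instances with listed ones.
# Reset clears all parameters, then reapplies the cached values, effectively dropping listed overrides; its save node is unwired.
# Select leaves assets unchanged and syncs the Content Browser to collected instances.
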EditorUtilityWidgetBlueprint /Game/ArtTools/Tech_Art/MI_Modify/EUW_ModMI
graphs [EventGraph {nodes [
   "On Clicked (EditorUtilityButton_41)"() {is=title K2Node_ComponentBoundEvent}
   > n1 = Set doSelect(doSelect:bool=false) -> (Output_Get:bool)
   n2 = For Each Loop(Exec=[@n26.then, @n1.then], Array:Array of Object References=Get Objects() -> (Objects:Array of Object References)) -> (Array Element:Object Reference)
   > Clear!(execute, TargetMap:Map of Names to Linear Color Structures=Get vectors() -> (vectors:Map of Names to Linear Color Structures)) -> (then) {is=title K2Node_CallFunction}
   > Clear(TargetMap:Map of Names to Texture Object References=Get textures() -> (textures:Map of Names to Texture Object References))
   > Clear(TargetMap:"Map of Names to Float (double-precision)s"=Get floats() -> (floats:"Map of Names to Float (double-precision)s"))
   > n3 = Cast To MaterialInstanceConstant(Object=@n2)
   > n4 = Sequence!(execute) -> (then_0, then_1, then_2:Exec)
   > n5 = For Each Loop!(
    Exec
    Array:Array of Scalar Parameter Value Structures=Get ScalarParameterValues!(self:Material Instance Object Reference=@n3) -> (ScalarParameterValues:Array of Scalar Parameter Value Structures)
   ) -> (
    LoopBody
    Array Element_ParameterInfo_Name:Name
    Array Element_ParameterInfo_Association:EMaterialParameterAssociation Enum
    Array Element_ParameterInfo_Index:int
    Array Element_ParameterValue:float=~
    Array Index
    Completed
   )
   n6 = Get ResetParams() -> (ResetParams:Set of Names)
   > n7 = Branch(Condition=Contains Item!(TargetSet:Set of Names=@n6, ItemToFind:"Name (by ref)"=@n5) -> (ReturnValue:bool) {is=title K2Node_CallFunction})
   Add Unique(
    execute=[@n25.then, @n7.then, @n10.then]
    TargetArray:Array of Material Instance Constant Object References=Get MICsToSelect() -> (MICsToSelect:Array of Material Instance Constant Object References)
    NewItem:"Material Instance Constant Object Reference (by ref)"=@n3
   )
   n8 = Get floats()
   Add!(
    execute=@n7.else
    TargetMap:"Map of Names to Float (double-precision)s"=@n8
    Key:"Name (by ref)"=@n5
    Value:"Float (double-precision) (by ref)"=@n5.Array Element_ParameterValue
   ) -> (then) {is=title K2Node_CallFunction}
   > PrintString(InString=Append!(A="To String (Name)"(InName=@n5), B=" - ", C:String="To String (Float)"(InDouble=@n5.Array Element_ParameterValue)) -> (ReturnValue))
   n9 = For Each Loop(
    Exec=@n4.then_1
    Array:Array of Texture Parameter Value Structures=Get TextureParameterValues!(self:Material Instance Object Reference=@n3) -> (TextureParameterValues:Array of Texture Parameter Value Structures)
   ) -> (Array Element_ParameterValue:Texture Object Reference)
   > n10 = Branch(Condition=Contains Item(TargetSet=@n6, ItemToFind=@n9))
   n11 = Get textures()
   Add(
    execute=@n10.else
    TargetMap:Map of Names to Texture Object References=@n11
    Key=@n9
    Value:"Texture Object Reference (by ref)"=@n9.Array Element_ParameterValue
   )
   > PrintString(InString=Append(A="To String (Name)"(InName=@n9), B=" - ", C=GetDisplayName(Object=@n9.Array Element_ParameterValue)))
   n12 = For Each Loop(
    Exec=@n4.then_2
    Array:Array of Vector Parameter Value Structures=Get VectorParameterValues!(self:Material Instance Object Reference=@n3) -> (VectorParameterValues:Array of Vector Parameter Value Structures)
   ) -> (Array Element_ParameterValue:Linear Color Structure)
   n13 = Branch(execute=@n12.Completed, Condition=Get doSelect() -> (doSelect:bool))
   n14 = Get PathsToSelect() -> (PathsToSelect:Array of Strings)
   > Clear(TargetArray=@n14) {is=title K2Node_CallArrayFunction}
   > n15 = For Each Loop!(Exec, Array:Array of Material Instance Constant Object References=Get MICsToSelect()) -> (LoopBody, Array Element:Material Instance Constant Object Reference, Array Index, Completed)
   SyncBrowserToObjects(execute=@n15.Completed, AssetPaths=@n14)
   n16 = GetPathNameForLoadedAsset(execute=@n15.LoopBody, LoadedAsset=@n15)
   > Add(TargetArray=@n14, NewItem=@n16) {is=title K2Node_CallArrayFunction}
   ClearAllMaterialInstanceParameters!(execute=@n13.else, Instance:Material Instance Constant Object Reference=@n3) -> (then)
   > n17 = Sequence()
   > n18 = Keys!(execute, TargetMap:"Map of Names to Float (double-precision)s"=@n8) -> (then, Keys:Array of Names)
   > n19 = For Each Loop(Array:Array of Names=@n18) -> (Array Element:Name)
   > SetMaterialInstanceScalarParameterValue!(
    execute
    Instance:Material Instance Constant Object Reference=@n3
    ParameterName:Name=@n19
    Value:float=Find(TargetMap:"Map of Names to Float (double-precision)s"=@n8, Key:"Name (by ref)"=@n19) -> (Value:double)
    Association:EMaterialParameterAssociation Enum=GlobalParameter
   ) -> (then, ReturnValue:bool)
   n20 = Keys(execute=@n17.then_1, TargetMap:Map of Names to Texture Object References=@n11)
   > n21 = For Each Loop(Array=@n20)
   > SetMaterialInstanceTextureParameterValue!(
    execute
    Instance:Material Instance Constant Object Reference=@n3
    ParameterName:Name=@n21
    Value:Texture Object Reference=Find(TargetMap:Map of Names to Texture Object References=@n11, Key=@n21) -> (Value:Texture Object Reference)
    Association:EMaterialParameterAssociation Enum=GlobalParameter
   ) -> (then, ReturnValue:bool)
   n22 = Get vectors()
   n23 = Keys(execute=@n17.then_2, TargetMap:Map of Names to Linear Color Structures=@n22)
   > n24 = For Each Loop(Array=@n23)
   > SetMaterialInstanceVectorParameterValue!(
    execute
    Instance:Material Instance Constant Object Reference=@n3
    ParameterName:Name=@n24
    Value:Linear Color Structure=Find(TargetMap:Map of Names to Linear Color Structures=@n22, Key=@n24) -> (Value:Linear Color Structure)
    Association:EMaterialParameterAssociation Enum=GlobalParameter
   ) -> (then, ReturnValue:bool)
   n25 = Branch(execute=@n12.LoopBody, Condition=Contains Item(TargetSet=@n6, ItemToFind=@n12))
   Add(
    execute=@n25.else
    TargetMap:Map of Names to Linear Color Structures=@n22
    Key=@n12
    Value:"Linear Color Structure (by ref)"=@n12.Array Element_ParameterValue
   ) {is=title K2Node_CallFunction}
   > PrintString(InString=Append(
     A="To String (Name)"(InName=@n12)
     B=" - "
     C="To String (LinearColor)"!(InColor:Linear Color Structure=@n12.Array Element_ParameterValue) -> (ReturnValue:String)
    ))
   "On Clicked (EditorUtilityButton_Select)"() {is=title K2Node_ComponentBoundEvent}
   > Set doSelect(doSelect=true)
   > n26 = Clear(TargetArray:Array of Material Instance Constant Object References=Get MICsToSelect())
   Event PreConstruct()
   > SetObject(
    self=[
     Get DetailsView_objects() -> (DetailsView_objects:Details View Object Reference)
     Get DetailsView_skip() -> (DetailsView_skip:Details View Object Reference)
    ]
    NewObject=Self-Reference()
   )
   SaveLoadedAsset(AssetToSave=@n3, bOnlyIfIsDirty=false)
   Event Construct()
   Event Tick()
  ]}]
variables [
 doSelect {type=boolean}
 floats {type="map<number>", default {}}
 MICsToSelect {type="array<MaterialInstanceConstant>", default []}
 Objects {type="array<Object>", default []}
 PathsToSelect {type="array<string>", default []}
 ResetParams {type="array<string>", default [OffsetImage_X, OffsetImage_Y, OffsetImage_Y_Compensation, ZoomImage_Percent]}
 textures {type="map<Texture>", default {}}
 vectors {type="map<LinearColor>", default {}}
]
widgets {
 info {rootWidgetClass=&VerticalBox, widgetCount=12}
 widgets [
  VerticalBox_0 {widgetClassPath=&VerticalBox, widget=VerticalBox_0}
  HorizontalBox_38 {widgetClassPath=&HorizontalBox, parent=VerticalBox_0, widget=HorizontalBox_38, slot {class=VerticalBoxSlot}}
  EditorUtilityButton_41 {
   widgetClassPath=&EditorUtilityButton
   parent=HorizontalBox_38
   widget=EditorUtilityButton_41
   slot {class=HorizontalBoxSlot, properties {size {sizeRule=Fill, value=1}}}
  }
  TextBlock_130 {
   widgetClassPath=&TextBlock
   parent=EditorUtilityButton_41
   widget=TextBlock_130
   properties {text=Reset MI Params}
   slot {class=ButtonSlot}
  }
  HorizontalBox_1 {widgetClassPath=&HorizontalBox, parent=VerticalBox_0, widget=HorizontalBox_1, slot {class=VerticalBoxSlot}}
  EditorUtilityButton_Select {
   widgetClassPath=&EditorUtilityButton
   parent=HorizontalBox_1
   widget=EditorUtilityButton_Select
   slot {class=HorizontalBoxSlot, properties {size {sizeRule=Fill, value=1}}}
  }
  TextBlock {
   widgetClassPath=&TextBlock
   parent=EditorUtilityButton_Select
   widget=TextBlock
   properties {text="Select MIs with \"Reset Params\" set"}
   slot {class=ButtonSlot}
  }
  HorizontalBox_307 {widgetClassPath=&HorizontalBox, parent=VerticalBox_0, widget=HorizontalBox_307, slot {class=VerticalBoxSlot}}
  DetailsView_skip {
   widgetClassPath=&DetailsView
   parent=HorizontalBox_307
   widget=DetailsView_skip
   properties {categoriesToShow [base]}
   slot {class=HorizontalBoxSlot, properties {size {sizeRule=Fill, value=1}}}
  }
  HorizontalBox {widgetClassPath=&HorizontalBox, parent=VerticalBox_0, widget=HorizontalBox, slot {class=VerticalBoxSlot}}
  AthenaScrollBox_77 {
   widgetClassPath=/Script/FortniteUI.AthenaScrollBox
   parent=HorizontalBox
   widget=AthenaScrollBox_77
   properties {
    widgetBarStyle {
     thickness=8
     draggedThumbImage {
      bIsDynamicallyLoaded=false
      drawAs=RoundedBox
      imageType=NoImage
      mirroring=NoMirror
      resourceName=None
      resourceObject=null
      tiling=NoTile
      imageSize=<0 0>
      margin=<ltrb 0 0 0 0>
      outlineSettings {
       bUseBrushTransparency=false
       roundingType=FixedRadius
       width=0
       color {colorUseRule=UseColor_Specified, specifiedColor=<rgba 0 0 0 0>}
       cornerRadii=<4 4 4 4>
      }
      tintColor {colorUseRule=UseColor_Specified, specifiedColor=<rgba 0.2159 0.2159 0.2159 1>}
      uVRegion {bIsValid=false, max=<0 0>, min=<0 0>}
     }
     horizontalBackgroundImage {
      bIsDynamicallyLoaded=false
      drawAs=NoDrawType
      imageType=NoImage
      mirroring=NoMirror
      resourceName=None
      resourceObject=null
      tiling=NoTile
      imageSize=<0 0>
      margin=<ltrb 0 0 0 0>
      outlineSettings {
       bUseBrushTransparency=false
       roundingType=HalfHeightRadius
       width=0
       color {colorUseRule=UseColor_Specified, specifiedColor=<rgba 0 0 0 0>}
       cornerRadii=<0 0 0 0>
      }
      tintColor {colorUseRule=UseColor_Specified, specifiedColor=<rgba 1 1 1 1>}
      uVRegion {bIsValid=false, max=<0 0>, min=<0 0>}
     }
     horizontalBottomSlotImage {
      bIsDynamicallyLoaded=false
      drawAs=NoDrawType
      imageType=NoImage
      mirroring=NoMirror
      resourceName=None
      resourceObject=null
      tiling=NoTile
      imageSize=<0 0>
      margin=<ltrb 0 0 0 0>
      outlineSettings {
       bUseBrushTransparency=false
       roundingType=HalfHeightRadius
       width=0
       color {colorUseRule=UseColor_Specified, specifiedColor=<rgba 0 0 0 0>}
       cornerRadii=<0 0 0 0>
      }
      tintColor {colorUseRule=UseColor_Specified, specifiedColor=<rgba 1 1 1 1>}
      uVRegion {bIsValid=false, max=<0 0>, min=<0 0>}
     }
     horizontalTopSlotImage {
      bIsDynamicallyLoaded=false
      drawAs=NoDrawType
      imageType=NoImage
      mirroring=NoMirror
      resourceName=None
      resourceObject=null
      tiling=NoTile
      imageSize=<0 0>
      margin=<ltrb 0 0 0 0>
      outlineSettings {
       bUseBrushTransparency=false
       roundingType=HalfHeightRadius
       width=0
       color {colorUseRule=UseColor_Specified, specifiedColor=<rgba 0 0 0 0>}
       cornerRadii=<0 0 0 0>
      }
      tintColor {colorUseRule=UseColor_Specified, specifiedColor=<rgba 1 1 1 1>}
      uVRegion {bIsValid=false, max=<0 0>, min=<0 0>}
     }
     hoveredThumbImage {
      bIsDynamicallyLoaded=false
      drawAs=RoundedBox
      imageType=NoImage
      mirroring=NoMirror
      resourceName=None
      resourceObject=null
      tiling=NoTile
      imageSize=<0 0>
      margin=<ltrb 0 0 0 0>
      outlineSettings {
       bUseBrushTransparency=false
       roundingType=FixedRadius
       width=0
       color {colorUseRule=UseColor_Specified, specifiedColor=<rgba 0 0 0 0>}
       cornerRadii=<4 4 4 4>
      }
      tintColor {colorUseRule=UseColor_Specified, specifiedColor=<rgba 0.2159 0.2159 0.2159 1>}
      uVRegion {bIsValid=false, max=<0 0>, min=<0 0>}
     }
     normalThumbImage {
      bIsDynamicallyLoaded=false
      drawAs=RoundedBox
      imageType=NoImage
      mirroring=NoMirror
      resourceName=None
      resourceObject=null
      tiling=NoTile
      imageSize=<0 0>
      margin=<ltrb 0 0 0 0>
      outlineSettings {
       bUseBrushTransparency=false
       roundingType=FixedRadius
       width=0
       color {colorUseRule=UseColor_Specified, specifiedColor=<rgba 0 0 0 0>}
       cornerRadii=<4 4 4 4>
      }
      tintColor {colorUseRule=UseColor_Specified, specifiedColor=<rgba 0.0953 0.0953 0.0953 1>}
      uVRegion {bIsValid=false, max=<0 0>, min=<0 0>}
     }
     verticalBackgroundImage {
      bIsDynamicallyLoaded=false
      drawAs=NoDrawType
      imageType=NoImage
      mirroring=NoMirror
      resourceName=None
      resourceObject=null
      tiling=NoTile
      imageSize=<0 0>
      margin=<ltrb 0 0 0 0>
      outlineSettings {
       bUseBrushTransparency=false
       roundingType=HalfHeightRadius
       width=0
       color {colorUseRule=UseColor_Specified, specifiedColor=<rgba 0 0 0 0>}
       cornerRadii=<0 0 0 0>
      }
      tintColor {colorUseRule=UseColor_Specified, specifiedColor=<rgba 1 1 1 1>}
      uVRegion {bIsValid=false, max=<0 0>, min=<0 0>}
     }
     verticalBottomSlotImage {
      bIsDynamicallyLoaded=false
      drawAs=NoDrawType
      imageType=NoImage
      mirroring=NoMirror
      resourceName=None
      resourceObject=null
      tiling=NoTile
      imageSize=<0 0>
      margin=<ltrb 0 0 0 0>
      outlineSettings {
       bUseBrushTransparency=false
       roundingType=HalfHeightRadius
       width=0
       color {colorUseRule=UseColor_Specified, specifiedColor=<rgba 0 0 0 0>}
       cornerRadii=<0 0 0 0>
      }
      tintColor {colorUseRule=UseColor_Specified, specifiedColor=<rgba 1 1 1 1>}
      uVRegion {bIsValid=false, max=<0 0>, min=<0 0>}
     }
     verticalTopSlotImage {
      bIsDynamicallyLoaded=false
      drawAs=NoDrawType
      imageType=NoImage
      mirroring=NoMirror
      resourceName=None
      resourceObject=null
      tiling=NoTile
      imageSize=<0 0>
      margin=<ltrb 0 0 0 0>
      outlineSettings {
       bUseBrushTransparency=false
       roundingType=HalfHeightRadius
       width=0
       color {colorUseRule=UseColor_Specified, specifiedColor=<rgba 0 0 0 0>}
       cornerRadii=<0 0 0 0>
      }
      tintColor {colorUseRule=UseColor_Specified, specifiedColor=<rgba 1 1 1 1>}
      uVRegion {bIsValid=false, max=<0 0>, min=<0 0>}
     }
    }
    widgetStyle {
     barThickness=8
     bottomShadowBrush {
      bIsDynamicallyLoaded=false
      drawAs=Box
      imageType=FullColor
      mirroring=NoMirror
      resourceName=../../../Engine/Content/Slate/Common/ScrollBoxShadowBottom.png
      resourceObject=null
      tiling=NoTile
      imageSize=<16 8>
      margin=<ltrb 0.5 0 0.5 1>
      outlineSettings {
       bUseBrushTransparency=false
       roundingType=HalfHeightRadius
       width=0
       color {colorUseRule=UseColor_Specified, specifiedColor=<rgba 0 0 0 0>}
       cornerRadii=<0 0 0 0>
      }
      tintColor {colorUseRule=UseColor_Specified, specifiedColor=<rgba 1 1 1 1>}
      uVRegion {bIsValid=false, max=<0 0>, min=<0 0>}
     }
     horizontalScrolledContentPadding=<ltrb 0 0 1 0>
     leftShadowBrush {
      bIsDynamicallyLoaded=false
      drawAs=Box
      imageType=FullColor
      mirroring=NoMirror
      resourceName=../../../Engine/Content/Slate/Common/ScrollBoxShadowLeft.png
      resourceObject=null
      tiling=NoTile
      imageSize=<8 16>
      margin=<ltrb 1 0.5 0 0.5>
      outlineSettings {
       bUseBrushTransparency=false
       roundingType=HalfHeightRadius
       width=0
       color {colorUseRule=UseColor_Specified, specifiedColor=<rgba 0 0 0 0>}
       cornerRadii=<0 0 0 0>
      }
      tintColor {colorUseRule=UseColor_Specified, specifiedColor=<rgba 1 1 1 1>}
      uVRegion {bIsValid=false, max=<0 0>, min=<0 0>}
     }
     rightShadowBrush {
      bIsDynamicallyLoaded=false
      drawAs=Box
      imageType=FullColor
      mirroring=NoMirror
      resourceName=../../../Engine/Content/Slate/Common/ScrollBoxShadowRight.png
      resourceObject=null
      tiling=NoTile
      imageSize=<8 16>
      margin=<ltrb 0 0.5 1 0.5>
      outlineSettings {
       bUseBrushTransparency=false
       roundingType=HalfHeightRadius
       width=0
       color {colorUseRule=UseColor_Specified, specifiedColor=<rgba 0 0 0 0>}
       cornerRadii=<0 0 0 0>
      }
      tintColor {colorUseRule=UseColor_Specified, specifiedColor=<rgba 1 1 1 1>}
      uVRegion {bIsValid=false, max=<0 0>, min=<0 0>}
     }
     topShadowBrush {
      bIsDynamicallyLoaded=false
      drawAs=Box
      imageType=FullColor
      mirroring=NoMirror
      resourceName=../../../Engine/Content/Slate/Common/ScrollBoxShadowTop.png
      resourceObject=null
      tiling=NoTile
      imageSize=<16 8>
      margin=<ltrb 0.5 1 0.5 0>
      outlineSettings {
       bUseBrushTransparency=false
       roundingType=HalfHeightRadius
       width=0
       color {colorUseRule=UseColor_Specified, specifiedColor=<rgba 0 0 0 0>}
       cornerRadii=<0 0 0 0>
      }
      tintColor {colorUseRule=UseColor_Specified, specifiedColor=<rgba 1 1 1 1>}
      uVRegion {bIsValid=false, max=<0 0>, min=<0 0>}
     }
     verticalScrolledContentPadding=<ltrb 0 0 0 1>
    }
   }
   slot {class=HorizontalBoxSlot, properties {size {sizeRule=Fill, value=1}}}
  }
  DetailsView_objects {
   widgetClassPath=&DetailsView
   parent=AthenaScrollBox_77
   widget=DetailsView_objects
   properties {categoriesToShow [list]}
   slot {class=ScrollBoxSlot, properties {size {sizeRule=Fill, value=1}}}
  }
 ]
}
